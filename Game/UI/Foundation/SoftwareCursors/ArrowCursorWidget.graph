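# UserWidget that draws the arrow software cursor, going by its name and folder.
# A root canvas panel holds one top-left anchored Border, which shows the CursorArrow image at cursor size, offset from the canvas corner.
# It has no working logic: Construct and Tick are empty, and the GetBackground_0 function returns an unset brush with no binding to it shown.
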WidgetBlueprint /Game/UI/Foundation/SoftwareCursors/ArrowCursorWidget
graphs [
 EventGraph {nodes [
   Event Construct()
   Event Tick()
  ]}
 GetBackground_0 {nodes [
   GetBackground_0() {is=title K2Node_FunctionEntry}
   > Return Node!(execute, ReturnValue:Slate Brush Structure)
  ]}
]
variables []
widgets {
 info {inheritedWidgetCount=0, parentClass=&UserWidget, rootWidgetClass=&CanvasPanel, widgetCount=2}
 widgets [
  CanvasPanel_0 {widgetClassPath=&CanvasPanel, widget=CanvasPanel_0}
  Border_15 {
   widgetClassPath=&Border
   parent=CanvasPanel_0
   widget=Border_15
   properties {background {
     bIsDynamicallyLoaded=false
     drawAs=Image
     imageType=NoImage
     mirroring=NoMirror
     resourceName=None
     resourceObject=/Game/UI/Cursors/CursorArrow
     tiling=NoTile
     imageSize=<32 32>
     margin=<ltrb 0 0 0 0>
     outlineSettings {
      bUseBrushTransparency=false
      roundingType=HalfHeightRadius
      width=0
      color {colorUseRule=UseColor_Specified, specifiedColor=<rgba 0 0 0 0>}
      cornerRadii=<0 0 0 0>
     }
     tintColor {colorUseRule=UseColor_Specified, specifiedColor=<rgba 1 1 1 1>}
     uVRegion {bIsValid=false, max=<0 0>, min=<0 0>}
    }}
   slot {class=CanvasPanelSlot, properties {layoutData {alignment=<0 0>, anchors {maximum=<0 0>, minimum=<0 0>}, offsets=<ltrb 32 32 32 32>}}}
  }
 ]
}
