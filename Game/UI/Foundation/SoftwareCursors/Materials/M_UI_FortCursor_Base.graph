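# Base translucent UI material for Fortnite software cursors, drawing an icon from background and foreground MSDF textures, defaulting to the arrow cursor.
# Each texture's channels go through a median-style min/max decode, then a SmoothStep whose upper edge is widened by a Smoothing parameter.
# The resulting named-reroute masks drive nested lerps that blend in a fixed near-black color.
# The textures are parameters, so instances can swap cursor shapes.
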Material /Game/UI/Foundation/SoftwareCursors/Materials/M_UI_FortCursor_Base
graph {
 properties {
  bAutomaticallySetUsageInEditor=false
  bUsedWithStaticMesh=false
  blendMode=BLEND_Translucent
  materialDomain=MD_UI
  usedShadingModels=Default Lit
 }
 nodes [
  LinearInterpolate(A=LinearInterpolate(Alpha=NamedRerouteUsage()), B=Constant3Vector() {constant=<rgba 0.0048 0.004 0.0116 1>}, Alpha=NamedRerouteUsage())
  n1 = Add(B=ScalarParameter() {parameterName=Smoothing}) {constA=0.5, constB=0}
  n2 = TextureSampleParameter2D() {
   group=Distance Field
   parameterName=Background MSDF
   samplerType=SAMPLERTYPE_LinearColor
   texture=/Game/UI/Foundation/SoftwareCursors/Textures/T_UI_Icon_UIKit_FortCursorArrow_Background_MSDF
  }
  n3 = Min(A=@n2.R, B=@n2.R)
  n4 = Min(A=Max(A=@n2.R, B=@n2.R), B=@n2.B)
  NamedRerouteDeclaration(None=SmoothStep(Max=@n1, Value=Max(A=@n3, B=@n4)) {constMax=0.6, constMin=0.5}) {name=Icon Background Mask, nodeColor=<rgba 0.9688 0 1 1>}
  n5 = TextureSampleParameter2D() {
   group=Distance Field
   parameterName=Foreground MSDF
   samplerType=SAMPLERTYPE_LinearColor
   texture=/Game/UI/Foundation/SoftwareCursors/Textures/T_UI_Icon_UIKit_FortCursorArrow_Foreground_MSDF
  }
  n6 = Min(A=@n5.R, B=@n5.R)
  n7 = Min(A=Max(A=@n5.R, B=@n5.R), B=@n5.B)
  NamedRerouteDeclaration(None=SmoothStep(Max=@n1, Value=Max(A=@n6, B=@n7)) {constMax=0.6, constMin=0.5}) {name=Icon Foreground Mask, nodeColor=<rgba 0.9688 0 1 1>}
  NamedRerouteUsage()
 ]
}
